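# EQS query that picks an approach position for the Zero Point Sprite NPC around its blackboard Leader, the target.
# It generates a ring of nav-projected points, with radius driven by a game-difficulty curve.
# A batch pathfinding test discards points the querier cannot reach using the water-avoiding nav filter.
# Scoring favors distance from nearby AI pawns and, weighted higher, positions in front of the target via a dot test.
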EnvQuery /Sprites_ZeroPointSprite/NPCs/EQS/EQS_ZeroPointSprite_Approach_Target
options [{
  generator {
   name=EnvQueryGenerator_OnCircle_3
   class=EnvQueryGenerator_OnCircle
   properties {
    circleCenter=/Script/FortniteAIServer.FortQueryContext_BlackboardKeyLeader
    numberOfPoints=12
    circleRadius {
     dataBinding=&self:EnvQueryGenerator_OnCircle_3.FortAIDataProvider_FloatCurveOverGameDifficulty_0
     dataField=FloatValue
     defaultValue=400
    }
    projectionData {
     extentX=0
     navigationFilter=&NavFilter_NPC_AvoidAllWater_C
     postProjectionVerticalOffset=0
     projectDown=1024
    }
    traceData {navigationFilter=null}
   }
  }
  tests [
   FortQueryTest_PathfindingBatch_2 {
    class=FortQueryTest_PathfindingBatch
    properties {
     bDefineReferenceValue=false
     boolValue=true
     clampMaxType=None
     clampMinType=None
     context=&EnvQueryContext_Querier
     filterClass=&NavFilter_NPC_AvoidAllWater_C
     filterType=Match
     floatValueMax=1000
     floatValueMin=1000
     multipleContextFilterOp=AllPass
     multipleContextScoreOp=AverageScore
     navDataOverrideContext=null
     normalizationType=Absolute
     pathFromContext=true
     referenceValue=0
     scanRangeMultiplier=1.5
     scoreClampMax=0
     scoreClampMin=0
     scoringEquation=Constant
     scoringFactor=1
     skipUnreachable=true
     testComment=Discard points that required pathing into water
     testMode=PathExist
     testPurpose=Filter
     navFilterTag {tagName=None}
    }
   }
   EnvQueryTest_Distance_1 {
    class=EnvQueryTest_Distance
    properties {
     distanceTo=&FortQueryContext_NearbyAIPawns
     filterType=Range
     floatValueMin=0
     scoringEquation=Linear
     scoringFactor=1
     testComment=Prefer locations away from other AI pawns
     testMode=Distance3D
     testPurpose=Score
    }
   }
   EnvQueryTest_Dot_0 {
    class=EnvQueryTest_Dot
    properties {
     clampMaxType=None
     clampMinType=None
     scoreClampMax=0
     scoreClampMin=0
     scoringFactor=1.5
     testComment=Try to position in front of the target
     lineA {dirMode=TwoPoints, lineFrom=&FortQueryContext_BlackboardKeyLeader, lineTo=&EnvQueryContext_Item}
     lineB {dirMode=Rotation, rotation=&FortQueryContext_BlackboardKeyLeader}
    }
   }
  ]
 }]
